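# Animation Blueprint for the ApplePound medium male player body, driving its secondary motion and material-based cloth effect.
# The AnimGraph copies the parent mesh pose, additively offsets both sweatshirt support bones, then runs a Control Rig.
# The EventGraph feeds bone velocity and collision for arm and pelvis regions into WPO cloth parameters on base and overlay materials, with wind tied to skydiving state.
# It also computes remapped thigh yaw values relative to the pelvis.
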AnimBlueprint /BRCosmetics/Characters/Player/Male/Medium/Bodies/M_MED_ApplePound/Meshes/M_MED_ApplePound_AnimBP
target_skeleton=&/M_MED_ApplePound_Skeleton
graphs [
 AnimGraph {nodes [
   n1 = "Transform (Modify) Bone - Bone: sweatshirt_support_L"!(ComponentPose=Local To Component(LocalPose=Copy Pose From Mesh() {bUseAttachedParent=true}), Alpha=1, Translation="5,0,0", Rotation="0,0,20") -> (Pose) {
    rotationMode=BMM_Additive
    rotationSpace=BCS_BoneSpace
    translationMode=BMM_Additive
    translationSpace=BCS_BoneSpace
    alphaBoolBlend {blendOption=Linear}
    boneToModify {boneName=sweatshirt_support_L}
    rotation {roll=20}
    translation {x=5}
   }
   n2 = "Transform (Modify) Bone - Bone: sweatshirt_support_R"!(ComponentPose=@n1, Alpha=1, Translation="5,0,0", Rotation="0,0,-20") -> (Pose) {
    rotationMode=BMM_Additive
    rotationSpace=BCS_BoneSpace
    translationMode=BMM_Additive
    translationSpace=BCS_BoneSpace
    alphaBoolBlend {blendOption=Linear}
    boneToModify {boneName=sweatshirt_support_R}
    rotation {roll=-20}
    translation {x=5}
   }
   n3 = Control Rig(Source=Component To Local(ComponentPose=@n2)) {
    alphaBoolBlend {blendOption=Linear}
    controlRigAssetReference {blueprintRigClass=&/M_MED_ApplePound_ControlRig.M_MED_ApplePound_ControlRig_C}
   }
   Output Pose(Result=@n3) {layerGroup=None}
  ]}
 EventGraph {nodes [
   n1 = Event BlueprintUpdateAnimation()
   > Set Delta Time X(Delta Time X=@n1.DeltaTimeX)
   > n2 = Sequence()
   > Is Valid(InputObject=Get PlayerPawn() -> (PlayerPawn:Fort Player Pawn Athena Object Reference))
   n3 = Get bIsSkyDiving()
   > n4 = Branch(Condition=OR Boolean(A=Get bIsInFrontEnd(), B=NOT Boolean(A=@n3)))
   n5 = Get Mat Instance Dynamic()
   n6 = Get Overlay Mat Instance Dynamic()
   > SetScalarParameterValue(self=[@n5, @n6], ParameterName=ClothSim_WindNoiseSpeed, Value=0)
   > SetScalarParameterValue(self=[@n5, @n6], ParameterName=ClothSim_TurbulenceNoiseTileDivisor, Value=0)
   n7 = Get Mat Instance Dynamic()
   n8 = Get Overlay Mat Instance Dynamic()
   SetScalarParameterValue(execute=@n4.else, self=[@n7, @n8], ParameterName=ClothSim_WindNoiseSpeed, Value=50)
   > SetScalarParameterValue(self=[@n7, @n8], ParameterName=ClothSim_TurbulenceNoiseTileDivisor, Value=25)
   > Set WPO Globals Function(
    Material=Get Mat Instance Dynamic()
    Overlay Material=Get Overlay Mat Instance Dynamic()
    Skeletal Mesh=Get SkelMesh()
    "Local Noise Transform  Location"=Get Local Noise Transform Location() -> (Local Noise Transform Location:Vector)
    Local Noise Offset=Get Local Noise Offset()
   )
   > Set WPO Regional Function!(
    execute
    Material:Material Instance Dynamic Object Reference=Get Mat Instance Dynamic()
    Overlay Material:Material Instance Dynamic Object Reference=Get Overlay Mat Instance Dynamic()
    Skeletal Mesh:Skeletal Mesh Component Object Reference=Get SkelMesh()
    Velocity Transform:Name=upperarm_l
    Velocity Interp Speed:double=Get Velocity Interp Speed()
    Collision Transform:Name=lowerarm_l
    Collision Transform Axis:EAxis Enum
    Invert Collision Transform Axis:bool=false
    Delta Time X:double=Get Delta Time X()
    Previous Velocity Transform Location:"Vector (by ref)"=Get R1 Velocity Transform Location()
    Previous Velocity Transform Velocity:"Vector (by ref)"=Get R1 Velocity Transform Velocity()
    Region:int=1
   ) -> (then)
   > Set WPO Regional Function(
    Material=Get Mat Instance Dynamic()
    Overlay Material=Get Overlay Mat Instance Dynamic()
    Skeletal Mesh=Get SkelMesh()
    Velocity Transform=pelvis
    Velocity Interp Speed=Get Velocity Interp Speed()
    Collision Transform=spine_01
    Invert Collision Transform Axis=true
    Delta Time X=Get Delta Time X()
    Previous Velocity Transform Location=Get R2 Velocity Transform Location()
    Previous Velocity Transform Velocity=Get R2 Velocity Transform Velocity()
    Region=2
   )
   > n9 = Branch(Condition=Get bIsSkydiveFloating())
   > n10 = Set WPO Pho Wind Intensity(WPO Pho Wind Intensity=1)
   SetScalarParameterValue(
    execute=[@n11.then, @n10.then]
    self=[Get Overlay Mat Instance Dynamic(), Get Mat Instance Dynamic()]
    ParameterName=ClothSim_WindStrength
    Value=Get WPO Pho Wind Intensity()
   )
   n11 = Set WPO Pho Wind Intensity(execute=@n9.else, WPO Pho Wind Intensity=0)
   n12 = GetOwnerSkeletalMeshComponent()
   n13 = GetDeltaTransformFromRefPose!(self=@n12, BoneName=thigh_l, BaseName=pelvis) -> (ReturnValue_Location:Vector, ReturnValue_Rotation:Rotator, ReturnValue_Scale:Vector)
   Set thigh_l Yaw(execute=@n2.then_1, thigh_l Yaw:float=MapRangeClamped(Value=BreakRotator(InRot=@n13.ReturnValue_Rotation).Yaw, InRangeA=-20, InRangeB=-90)) -> (Output_Get:float)
   n14 = GetDeltaTransformFromRefPose(self=@n12, BoneName=thigh_r, BaseName=pelvis)
   > Set thigh_r Yaw(thigh_r Yaw:float=MapRangeClamped(Value=BreakRotator(InRot=@n14.ReturnValue_Rotation).Yaw, InRangeA=-20, InRangeB=-90)) -> (Output_Get:float)
   Event BlueprintInitializeAnimation()
   > Set SkelMesh(SkelMesh=GetOwnerSkeletalMeshComponent())
   n15 = TryGetPawnOwner()
   > Is Valid(InputObject=@n15)
   > n16 = Cast To FortPlayerPawnAthena(Object=@n15)
   > Set PlayerPawn(PlayerPawn:Fort Player Pawn Athena Object Reference=@n16) -> (Output_Get:Fort Player Pawn Athena Object Reference)
   Event OnCharacterCustomizationLoaded()
   > n17 = Set SkelMesh(SkelMesh=GetPartSkeletalMeshComponent())
   > n18 = Get Dynamic Material Instance(SkeletalMeshComponent=@n17)
   > Set Mat Instance Dynamic(Mat Instance Dynamic=@n18)
   > n19 = GetOrCreateDynamicOverlayMaterial(MeshComponent=@n17, MaterialName=None)
   > Set Overlay Mat Instance Dynamic(Overlay Mat Instance Dynamic=@n19)
  ]}
]
variables [
 Delta Time X {type=number}
 Local Noise Offset {type=Vector}
 Local Noise Transform Location {type=Vector}
 Mat Instance Dynamic {type=MaterialInstanceDynamic}
 Overlay Mat Instance Dynamic {type=MaterialInstanceDynamic}
 PlayerPawn {type=FortPlayerPawnAthena}
 R1 Velocity Transform Location {type=Vector}
 R1 Velocity Transform Velocity {type=Vector}
 R2 Velocity Transform Location {type=Vector}
 R2 Velocity Transform Velocity {type=Vector}
 SkelMesh {type=SkeletalMeshComponent}
 thigh_l Yaw {type=number}
 thigh_r Yaw {type=number}
 Velocity Interp Speed {type=number}
 WPO Pho Wind Intensity {type=number}
]
